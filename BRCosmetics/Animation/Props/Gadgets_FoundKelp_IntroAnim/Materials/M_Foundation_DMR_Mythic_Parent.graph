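# Parent skeletal-mesh material for the Foundation mythic DMR rifle, adding a fake point-light glow with plasma noise to the emissive.
# Custom primitive data drives a Charge Phase value, either a direct charge amount or a timed, invertible curve-atlas easing for single shots.
# The masked sphere glow sits between the character PreFX and PostFX functions, followed by scope and anti-camera-clipping functions.
# A sine/cosine vibration offset is also computed, but World Position Offset shows no connection.
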Material /BRCosmetics/Animation/Props/Gadgets_FoundKelp_IntroAnim/Materials/M_Foundation_DMR_Mythic_Parent
graph {
 properties {
  bAutomaticallySetUsageInEditor=false
  bUseMaterialAttributes=true
  bUsedWithMeshDeformer=true
  bUsedWithMorphTargets=true
  bUsedWithSkeletalMesh=true
  pixelDepthOffsetMode=PDOM_Legacy
  shadingModel=MSM_FromMaterialExpression
  usedShadingModels=Default Lit | Subsurface | Clear Coat | Subsurface Profile | Cloth
 }
 nodes [
  n1 = LocalPosition() {localOrigin=Primitive}
  LinearInterpolate(A=@n1.XYZ, B=ActorPositionWS(), Alpha=Saturate(None=ScalarParameter() {parameterName=Is Scope?}))
  n2 = Multiply(A=Time(), B=ScalarParameter() {defaultValue=8, group=Foundation Rifle, parameterName=SpeedVibration})
  n3 = AppendVector(A=Sine(None=@n2), B=AppendVector(A=Sine(None=@n2), B=Cosine(None=@n2)))
  NamedRerouteDeclaration(None=Multiply(A=Normalize(VectorInput=@n3), B=ScalarParameter() {defaultValue=0.4, group=Foundation Rifle, parameterName=VibrationIntensity})) {name=Vibration Light, nodeColor=<rgba 1 0 0.4219 1>}
  n4 = Time()
  n5 = ScalarParameter() {bUseCustomPrimitiveData=true, defaultValue=500000000, parameterName=Game Time, primitiveDataIndex=2}
  n6 = ScalarParameter() {defaultValue=0.1, parameterName=Rift Blink Shoot Up - Duration}
  n7 = ScalarParameter() {defaultValue=0.1, parameterName=Rift Blink Shoot Down - Duration}
  n8 = ScalarParameter() {bUseCustomPrimitiveData=true, parameterName=Invert Time, primitiveDataIndex=4}
  NamedRerouteDeclaration(None=Divide(A=Subtract(A=@n4, B=@n5), B=LinearInterpolate(A=@n6, B=@n7, Alpha=@n8))) {name=Timestamp, nodeColor=<rgba 0.1016 0 1 1>}
  n9 = NamedRerouteUsage()
  n10 = CurveAtlasRowParameter(CurveTime=LinearInterpolate(A=@n9, B=OneMinus(None=@n9), Alpha=ScalarParameter() {bUseCustomPrimitiveData=true, parameterName=Invert Time, primitiveDataIndex=4})) {
   atlas=/Game/Athena/LinearColorCurves/VFX_CurveAtlas
   curve=/Game/Athena/LinearColorCurves/Curve_Burst_Up_Slow_Down
   defaultValue=5
   parameterName=Easing Curve
  }
  NamedRerouteDeclaration(None=LinearInterpolate(
    A=ScalarParameter() {bUseCustomPrimitiveData=true, parameterName=Charge Amount, primitiveDataIndex=1}
    B=@n10.R
    Alpha=ScalarParameter() {bUseCustomPrimitiveData=true, parameterName=Single Shot, primitiveDataIndex=5}
   )) {name=Charge Phase, nodeColor=<rgba 0 0.5781 1 1>}
  ObjectPositionWS()
  n11 = MaterialFunctionCall() {materialFunction=&MF_CharacterParent_2023_PreFX}
  n12 = GetMaterialAttributes(None=@n11.Result) {attributeGetTypes [#1, #2]}
  n13 = VectorParameter() {group=Foundation Rifle, parameterName=Sphere Glow Light, defaultValue=<rgba 0.0813 0.6462 1 1>}
  n14 = LinearInterpolate(
   A=ScalarParameter() {defaultValue=1.5, group=Foundation Rifle, parameterName=Fake Point Light - Intensity}
   B=ScalarParameter() {defaultValue=2, group=Foundation Rifle, parameterName=FAke Point Light - Max}
   Alpha=NamedRerouteUsage()
  )
  n15 = ComponentMask(None=@n11.Tech Art Mask) {g=true}
  n16 = Constant3Vector() {constant=<rgba 0 0 -1 0>}
  n17 = Power(
   Base=DotProduct(A=VertexNormalWS(), B=Transform(None=@n16) {transformSourceType=TRANSFORMSOURCE_Local})
   Exp=ScalarParameter() {defaultValue=2.5, parameterName=Vertex Normal Fresnel}
  )
  n18 = LinearInterpolate(A=@n15, B=Multiply(A=@n15, B=LinearInterpolate(Alpha=@n17) {constA=0.05}), Alpha=Saturate(None=ScalarParameter() {parameterName=Is Scope?}))
  n19 = Add(A=@n1.XYZ, B=VectorParameter() {group=Foundation Rifle, parameterName=Fake Point Light - Offset, defaultValue=<rgba 0 7 -4 0>}.RGB)
  n20 = Power(
   Base=SphereMask(
    A=WorldPosition().XYZ
    B=TransformPosition(None=@n19) {transformType=TRANSFORMPOSSOURCE_World}
    Radius=ScalarParameter() {defaultValue=50, group=Foundation Rifle, parameterName=Fake Point Light - Radius}
    Hardness=ScalarParameter() {group=Foundation Rifle, parameterName=Fake Point Light - Hardness}
   )
   Exp=ScalarParameter() {defaultValue=5, group=Foundation Rifle, parameterName=Fake Point Light - Exp}
  )
  n21 = MaterialFunctionCall!(
   UV=Multiply(A=TextureCoordinate(), B=ScalarParameter() {defaultValue=10.2423, group=Foundation Rifle, parameterName=Plasma Tiling})
   Center Glow brightness
  ) {materialFunction=/Game/Global/Materials/MaterialGym/Materials/PlasmaBall/MF_PlasmaNoise}
  n22 = ScalarParameter() {defaultValue=8, group=Foundation Rifle, parameterName=Plasma Intensity}
  n23 = Multiply(
   A=LinearInterpolate(Alpha=NamedRerouteUsage()) {constA=0.1}
   B=MaterialFunctionCall!(Base=Multiply(A=@n18, B=@n20), Add=Multiply(A=@n21.Result, B=@n22)) {materialFunction=/Engine/Functions/Engine_MaterialFunctions02/Math/MultiplyAdd}.Result
  )
  n24 = SetMaterialAttributes!(
   MaterialAttributes=@n12.MaterialAttributes
   Emissive Color=LinearInterpolate(A=@n12.Emissive Color, B=Multiply(A=@n13.RGB, B=@n14), Alpha=Saturate(None=@n23))
   Roughness=Add(A=@n12.Roughness, B=ScalarParameter() {group=Foundation Rifle, parameterName=Roughness Boost})
  ) {attributeSetTypes [#1, #2]}
  n25 = MaterialFunctionCall!(In=MaterialFunctionCall!(Material Attributes=MaterialFunctionCall(Material=@n24, M=@n11.M, Pre Skinned Local Position=@n11.M, Distance Blend=@n11.M, Fresnel=@n11.M) {materialFunction=&MF_CharacterParent_2023_PostFX}.Result) {materialFunction=/Game/Packages/Fortnite_UniqueMaterials_1/Weapons/ADSScopeMaster}.Result) {materialFunction=/Game/Packages/Fortress_SharedMaterials/Base_Material_Functions/BaseCharacterShaderFunctions/MF_AntiCameraClipping}
  SetMaterialAttributes!(
   MaterialAttributes=GetMaterialAttributes(None=MaterialFunctionCall!(In=@n25.Result) {materialFunction=/Game/Packages/Fortress_SharedMaterials/Base_Material_Functions/MF_ScopeMasking}.Result).MaterialAttributes
   World Position Offset
   Opacity
   FirstPersonInterpolationAlpha=Constant() {r=1}
  ) {attributeSetTypes [#3, #4, #5]}
 ]
}
